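# Post-process-style cosmetic AnimBlueprint for the male medium Genius Blob body, which copies its pose from the attached parent mesh through the body's Control Rig.
# Additive bone edits then rotate dyn_pelvis_rotate and offset both forearms while the Plummet emote montage plays.
# Each update derives pelvis rotation from left-thigh yaw: roll outside skydiving, roll plus pitch under an open parachute, zero in freefall.
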AnimBlueprint /BRCosmetics/Characters/Player/Male/Medium/Bodies/M_MED_Genius_Blob/Meshes/M_MED_Genius_Blob_AnimBP
target_skeleton=&/M_MED_Genius_Blob_Skeleton
graphs [
 AnimGraph {nodes [
   Local Space Ref Pose()
   n1 = Control Rig(Source=Copy Pose From Mesh() {bUseAttachedParent=true}) {
    alphaBoolBlend {blendOption=Linear}
    controlRigAssetReference {blueprintRigClass=&/M_MED_Genius_Blob_CtrlRig.M_MED_Genius_Blob_CtrlRig_C}
   }
   n2 = "Transform (Modify) Bone - Bone: dyn_pelvis_rotate"!(
    ComponentPose=Local To Component(LocalPose=@n1)
    bAlphaBoolEnabled:bool=True
    Translation=Get backpack_translation()
    Rotation=Get pelvis_rotate() -> (pelvis_rotate:Rotator)
    Scale="1,1,1"
   ) -> (Pose) {alphaInputType=Bool, rotationMode=BMM_Additive, alphaBoolBlend {blendOption=Linear}, boneToModify {boneName=dyn_pelvis_rotate}}
   n3 = Get isCustomEmotePlaying() -> (isCustomEmotePlaying:double)
   n4 = "Transform (Modify) Bone - Bone: lowerarm_r"(ComponentPose=@n2, Alpha=@n3, Translation="0,0,0", Rotation="0,-20,0", Scale="1,1,1") {rotationMode=BMM_Additive, alphaBoolBlend {blendOption=Linear}, boneToModify {boneName=lowerarm_r}, rotation {yaw=-20}}
   n5 = "Transform (Modify) Bone - Bone: lowerarm_l"(ComponentPose=@n4, Alpha=@n3, Translation="0,0,0", Rotation="0,20,0", Scale="1,1,1") {rotationMode=BMM_Additive, alphaBoolBlend {blendOption=Linear}, boneToModify {boneName=lowerarm_l}, rotation {yaw=20}}
   Output Pose(Result=Component To Local(ComponentPose=@n5)) {layerGroup=None}
  ]}
 EventGraph {nodes [
   Event BlueprintUpdateAnimation()
   n1 = GetAnimInstance(self=GetOwnerSkeletalMeshComponent())
   > Is Valid(InputObject=@n1)
   > Set isCustomEmotePlaying(isCustomEmotePlaying:double=SelectFloat(A=1, B=0, bPickA=Montage_IsPlaying(self=@n1, Montage=/BRCosmetics/Animation/Game/MainPlayer/Emotes/Plummet/Emote_Plummet_CMM_M))) -> (Output_Get:double)
   > n2 = Branch(Condition=Get bIsSkyDiving())
   > n3 = Branch(Condition=Get bIsParachuteOpened())
   n4 = GetOwnerSkeletalMeshComponent()
   n5 = BreakRotator(InRot=BreakTransform(InTransform=GetDeltaTransformFromRefPose(self=@n4, BoneName=thigh_l, BaseName=pelvis)).Rotation)
   n6 = MapRangeClamped(Value=@n5.Yaw, InRangeA=-90, InRangeB=0, OutRangeA=-30, OutRangeB=0)
   > Set pelvis_rotate(pelvis_rotate:Rotator=MakeRotator(Roll=@n6, Pitch=MapRangeClamped(Value=@n5.Yaw, InRangeA=-90, InRangeB=90, OutRangeA=-25, OutRangeB=30))) -> (Output_Get:Rotator)
   Set pelvis_rotate(execute=@n3.else, pelvis_rotate="0, 0, 0")
   Set pelvis_rotate(execute=@n2.else, pelvis_rotate=MakeRotator(Roll=@n6))
  ]}
]
variables [
 backpack_translation {type=Vector}
 isCustomEmotePlaying {type=number}
 pelvis_rotate {type=Rotator}
 pelvis_rotate_temp {type=Rotator}
]
